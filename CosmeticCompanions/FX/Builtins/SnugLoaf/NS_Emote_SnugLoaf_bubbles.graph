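# Niagara bubble effect for the SnugLoaf companion emote, with mesh bubbles that pop into sprite droplets.
# Bubbles spawns short-lived meshes in curve-driven pulses, animating scale and material parameters, with User.bubbles_location blending a position offset.
# A WriteDeath module feeds death positions through a system array and ReadDeath step, letting BurstStaticBubbles spawn small falling droplet bursts there.
# System-level colors and time-of-day scalars adjust the look.
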NiagaraSystem /CosmeticCompanions/FX/Builtins/SnugLoaf/NS_Emote_SnugLoaf_bubbles
dependencies {
 usedDataInterfaces [
  &NiagaraDataInterfaceCurve
  &NiagaraDataInterfaceArrayPosition
  &NiagaraDataInterfacePlatformSet
 ]
 usedDynamicInputs [
  /Game/Effects/Niagara/DynamicInputs/Logic/FN_PickLinearColorDependOnPlatform
  &FNTimeOfDayColor
  &RandomRangeLinearColor
 ]
 usedModules [
  &self:SystemSpawnScript.SystemScriptSource.SystemScriptGraph.NiagaraNodeAssignment_6.SetVariables_#1
  &self:SystemSpawnScript.SystemScriptSource.SystemScriptGraph.NiagaraNodeAssignment_7.SetVariables_#2
  &self:SystemSpawnScript.SystemScriptSource.SystemScriptGraph.NiagaraNodeAssignment_8.SetVariables_#3
  &self:SystemSpawnScript.SystemScriptSource.SystemScriptGraph.NiagaraNodeAssignment_9.SetVariables_#4
  &SystemState
  &self:ReadDeath
  &self:SystemSpawnScript.SystemScriptSource.SystemScriptGraph.NiagaraNodeAssignment_5.SetVariables_#5
 ]
}
emitters {
 Bubbles {
  emitterUpdateScript [
   &EmitterState(Life Cycle Mode=::NewEnumerator1, Loop Behavior=::NewEnumerator1, Loop Delay=4, Loop Duration=3.4, UseLoopDelay=false)
   &SpawnRate(
    SpawnRate=&FloatFromCurve(
     FloatCurve=DataInterface() {propertyValues {Curve {keys [<at 0 7 RCIM_Constant>,<at 0.04 2 RCIM_Constant>,<at 0.3204 5 RCIM_Constant>,<at 0.3439 0 RCIM_Constant>,<at 0.87 7 RCIM_Constant>]}}}
     CurveIndex=$Emitter.NormalizedLoopAge
     Scale Curve=5
    )
    Use Spawn Probability=false
   )
  ]
  particleSpawnScript [
   /SetVariables(Particles.VisibilityTag=&SpawnGroup()) {isSetParameters=true}
   &InitializeParticle(
    Color=$System.Bubble TOD
    Lifetime Max=0.3
    Lifetime Min=0.15
    Lifetime Mode=::NewEnumerator1
    Mesh Scale Mode=::NewEnumerator1
    Mesh Uniform Scale Max=0.065
    Mesh Uniform Scale Min=0.025
    Position Offset:NiagaraPosition=$System.Initial Position
    Sprite Rotation Mode=::NewEnumerator2
    UsePositionOffset=true
    Write Material Random=true
   )
   &ShapeLocation(
    Disc Coverage=0.2262
    MaskSpawnGroup=false
    Offset=<-5 0 0>
    Offset Mode=::NewEnumerator0
    Override Seed=false
    Ring Radius=5
    Shape Primitive=::NewEnumerator4
    UseOwnerScale=true
   )
   &CameraOffset(Camera Offset Amount=11)
  ]
  particleUpdateScript [
   &ParticleState()
   &ScaleMeshSize(Scale Factor=&VectorFromFloat(Value=&FloatFromCurve(FloatCurve=DataInterface() {propertyValues {Curve {keys [<at 0 0.2>,<at 0.8 1>,<at 1 0>]}}}, Scale Curve=1.8267)))
   &DynamicMaterialParameters(
    Index 0 Param 1=&FloatFromCurve(FloatCurve=DataInterface() {propertyValues {Curve {keys [
         <at 0 0>
         {interpMode=RCIM_Cubic, leaveTangent=1.2072, tangentMode=RCTM_Break, time=0.55, value=0}
         {arriveTangent=3.283, tangentMode=RCTM_User, time=1, value=1}
        ]}}})
    Index 0 Param 4=1.1
    Module.Index0Param0Write=true
    Module.Index0Param1Write=true
    Module.Index0Param2Write=true
    Module.Index0Param3Write=true
   )
   &GravityForce(Gravity=&MakeVector(
     X=0
     Y=0
     Z=&FloatFromCurve(FloatCurve=DataInterface() {propertyValues {Curve {defaultValue=1, keys [{tangentMode=RCTM_User, time=0, value=0}, {tangentMode=RCTM_User, time=1, value=1}]}}})
    ))
   /SetVariables(
    Particles.SplashDynamicParamenter:Vector4f=&MakeVector4(
     X=&FloatFromCurve(FloatCurve=DataInterface() {propertyValues {Curve {keys [
          <at 0 1>
          {interpMode=RCIM_Cubic, leaveTangent=3.8935, tangentMode=RCTM_Break, time=0.8, value=1.5}
          {arriveTangent=171.8661, tangentMode=RCTM_User, time=1, value=17}
         ]}}})
     Y=0.5
     Z=0
     W=0
    )
    Particles.SplashIndex:NiagaraFloat=&RandomRangeFloat(Minimum=0, Maximum=3)
    Particles.SplashSize:Vector2f=&Vector2DFromFloat(Value=&FloatFromCurve(
      FloatCurve=DataInterface() {propertyValues {Curve {keys [<at 0 0>,<at 0.7 0>,<at 0.85 1>]}}}
      Scale Curve=&RandomRangeFloat(Minimum=15, Maximum=20)
     ))
   ) {isSetParameters=true}
   /SetVariables(Particles.bubble:LinearColor=<rgba 0.7782 1 0.9558 1>) {isSetParameters=true}
   &ScaleColor(
    Scale RGB=&VectorFromFloat(Value=&FNTimeOfDayScalar(Night Scale=2, Morning Scale=2, Day Scale=2, Evening Scale=2))
    ScaleA=true
    ScaleRGB=true
   )
   &self:WriteDeath_01(Position Array:NiagaraDataInterfaceArrayPosition=$System.DeathArray) {module=WriteDeath_01}
   /SetVariables(Particles.Update Position:NiagaraPosition=&AddVectorToPosition(
     Position=&Multiply_Vector(A=Linked() {linkedVariable {name=Particles.ShapeLocation.ShapePosition, type=NiagaraPosition}}, B=<1 1 1>)
     Vector=&Lerp_Vector(A=<0 15 0>, B=<0 -24 1>, Alpha=$User.bubbles_location)
    )) {isSetParameters=true}
   &SolveForcesAndVelocity(Clamp Velocity=false, Limit Acceleration=false)
  ]
  rendererClasses [&NiagaraMeshRendererProperties]
  renderers [{rendererClass=&NiagaraMeshRendererProperties, rendererIndex=0}]
 }
 BurstStaticBubbles {
  emitterUpdateScript [
   &EmitterState(Life Cycle Mode=::NewEnumerator1, Loop Behavior=::NewEnumerator1, Loop Duration Mode=::NewEnumerator1, UseLoopDelay=false)
   &SpawnPerFrame(
    Spawn:NiagaraBool=$System.DeathSpawn
    Spawn Count:NiagaraInt32=&RandomRangeInt(Minimum=2, Maximum=3)
    Spawn Group=1
    Use Spawn Probability=false
   ) {module=SpawnPerFrame001}
  ]
  particleSpawnScript [
   /SetVariables(Particles.VisibilityTag=&SpawnGroup()) {isSetParameters=true}
   &InitializeParticle(
    Color=$System.Droplet tod
    Lifetime=0.6
    Mesh Scale Mode=::NewEnumerator3
    Position=&TransformPosition(Position=$System.DeathPosition, Source Space=::Simulation, Destination Space=::Simulation)
    Position Mode=::NewEnumerator1
    Sprite Rotation Mode=::NewEnumerator2
    Sprite Size Mode=::NewEnumerator3
    Uniform Sprite Size=&RandomRangeFloat(Minimum=20, Maximum=50)
    UsePositionOffset=false
    Write Material Random=true
   )
   &DynamicMaterialParameters(
    Index 0 Param 1=5
    Index 0 Param 2=1
    Index 0 Param 3=1
    Index 0 Param 4=1
    Index 1 Param 1=6.667
    Index 1 Param 3=1
    Index 1 Param 4=1
    Module.Index0Float0Write=true
    Module.Index0Float1Write=true
    Module.Index0Float2Write=true
    Module.Index0Float3Write=true
    Module.Index1Float0Write=true
    Module.Index1Float1Write=true
    Module.Index1Float2Write=true
    Module.Index1Float3Write=true
    Write Parameter Index 1=::NewEnumerator1
   )
   /SetVariables(
    Particles.Drops Lifetime:NiagaraFloat=&RandomRangeFloat(Minimum=0.2, Maximum=0.5)
    Particles.Ring Lifetime=0.1
   ) {isSetParameters=true}
   /SetVariables(Particles.Drops Size:Vector2f=&RandomRangeVector2D(Minimum=<0.5 2.5>, Maximum=<1 3.5>)) {isSetParameters=true}
   &AddVelocity(
    Cone Axis=<0 0 1>
    Module.UseConeFalloff=false
    Module.UseVelDistribution=false
    Override Seed=false
    Velocity Mode=::NewEnumerator2
    Velocity Speed=&RandomRangeFloat(Minimum=50, Maximum=150)
   )
   &ShapeLocation(
    MaskSpawnGroup=false
    Offset:NiagaraPosition=$System.DeathPosition
    Offset Mode=::NewEnumerator0
    Override Seed=false
    Sphere Radius=0
    UseOwnerScale=true
   ) {enabled=false}
  ]
  particleUpdateScript [
   &ParticleState()
   /SetVariables(
    Particles.Update Drops Age:NiagaraFloat=&Divide_Float(A=$Particles.Age, B=$Particles.Initial.Drops Lifetime)
    Particles.Update Ring Age:NiagaraFloat=&Divide_Float(A=$Particles.Age, B=$Particles.Initial.Ring Lifetime)
   ) {isSetParameters=true}
   /SetVariables(
    Particles.Update Drops Linear Color:LinearColor=&MakeLinearColorFromVectorAndFloat(
     "Vector (RGB)"=&MakeVectorFromLinearColorRGB(Color=$System.Droplet tod)
     "Float (Alpha)"=&FloatFromCurve(
      FloatCurve=DataInterface() {propertyValues {Curve {keys [
          <at 0.2917 1 RCIM_Cubic>
          {arriveTangent=-0.4998, interpMode=RCIM_Cubic, leaveTangent=-0.4998, tangentMode=RCTM_User, time=0.7275, value=1}
          <at 1 0>
         ]}}}
      CurveIndex=$Particles.Update Drops Age
      Scale Curve=&FNTimeOfDayScalar(Night Scale=1, Morning Scale=0.9, Evening Scale=0.8)
     )
    )
    Particles.Update Drops Size:Vector2f=&Multiply_Vector2D(
     A=$Particles.Initial.Drops Size
     B=&Vector2DFromFloat(Value=&FloatFromCurve(
       FloatCurve=DataInterface() {propertyValues {Curve {keys [
           {arriveTangent=1.3698, interpMode=RCIM_Cubic, leaveTangent=1.3698, tangentMode=RCTM_User, time=0, value=0.3}
           {tangentMode=RCTM_User, time=1, value=1}
          ]}}}
       CurveIndex=$Particles.Update Drops Age
       Scale Curve=2
      ))
    )
   ) {isSetParameters=true}
   &GravityForce(Gravity=<0 0 -500>)
   &SolveForcesAndVelocity(Clamp Velocity=false, Limit Acceleration=false)
   &ScaleSpriteSizeBySpeed(
    Sample Scale Factor By Curve=true
    Scale Factor Curve=DataInterface() {propertyValues {Curve {defaultValue=0, keys [{tangentMode=RCTM_User, time=0, value=1}, {tangentMode=RCTM_User, time=1, value=0}]}}}
    Velocity Threshold=50
   )
  ]
  rendererClasses [&NiagaraSpriteRendererProperties]
  renderers [{rendererClass=&NiagaraSpriteRendererProperties}]
 }
}
summary {
 systemName=NS_Emote_SnugLoaf_bubbles
 emitters [
  {emitterName=Bubbles, rendererClasses [&NiagaraMeshRendererProperties]}
  {emitterName=BurstStaticBubbles, rendererClasses [&NiagaraSpriteRendererProperties]}
 ]
 userVariables [
  User.bIsFrontEndPreview {type=NiagaraBool}
  User.bubbles_location {type=NiagaraFloat, defaultValue=0, description=""}
 ]
}
