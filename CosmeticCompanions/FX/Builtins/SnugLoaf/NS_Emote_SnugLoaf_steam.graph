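# Niagara system for the SnugLoaf emote that layers steam, falling petals and liquid ripples around a cup, apparently a latte.
# Two steam emitters share a system color, while two petal emitters add forces, collision and mesh-oriented sprites.
# Ripples_merged follows the cup's skeletal mesh, using spawn groups to offset timed and continuous ripples across four sprite renderers.
# Time-of-day inputs adjust opacity and ripple color, with platform checks varying ripple color and petal collision.
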NiagaraSystem /CosmeticCompanions/FX/Builtins/SnugLoaf/NS_Emote_SnugLoaf_steam
dependencies {
 usedDataInterfaces [&NiagaraDataInterfaceCurve, &NiagaraDataInterfacePlatformSet, &NiagaraDataInterfaceSkeletalMesh]
 usedDynamicInputs []
 usedModules [
  &self:SystemSpawnScript.SystemScriptSource.SystemScriptGraph.NiagaraNodeAssignment_0.SetVariables_#1
  &self:SystemSpawnScript.SystemScriptSource.SystemScriptGraph.NiagaraNodeAssignment_2.SetVariables_#2
  &self:SystemSpawnScript.SystemScriptSource.SystemScriptGraph.NiagaraNodeAssignment_3.SetVariables_#3
  &self:SystemSpawnScript.SystemScriptSource.SystemScriptGraph.NiagaraNodeAssignment_4.SetVariables_#4
  &self:SystemSpawnScript.SystemScriptSource.SystemScriptGraph.NiagaraNodeAssignment_6.SetVariables_#5
  &self:SystemSpawnScript.SystemScriptSource.SystemScriptGraph.NiagaraNodeAssignment_7.SetVariables_#6
  &SystemState
  &self:SystemSpawnScript.SystemScriptSource.SystemScriptGraph.NiagaraNodeAssignment_1.SetVariables_#7
  &self:SystemSpawnScript.SystemScriptSource.SystemScriptGraph.NiagaraNodeAssignment_5.SetVariables_#8
 ]
}
emitters {
 Billow_steam001 {
  emitterUpdateScript [
   &EmitterState(Life Cycle Mode=::NewEnumerator1, Loop Behavior=::NewEnumerator1, Loop Delay=0.9, Loop Duration=11, UseLoopDelay=false)
   &SpawnRate(SpawnRate=15, Use Spawn Probability=false)
  ]
  particleSpawnScript [
   &InitializeParticle(
    Color=&MakeLinearColorFromVectorAndFloat("Vector (RGB)"=&MakeVectorFromLinearColorRGB(Color=$System.SteamColor), "Float (Alpha)"=0.1)
    Lifetime=&RandomRangeFloat(Minimum=0.85, Maximum=1.7)
    Position Offset=<0 10 15>
    Sprite Rotation Mode=::NewEnumerator2
    Sprite Size Max=<55 65>
    Sprite Size Min=<30 35>
    Sprite Size Mode=::NewEnumerator2
    Sprite UV Mode=::NewEnumerator2
    UsePositionOffset=true
    Write Material Random=true
   )
   &CameraOffset(Camera Offset Amount=10)
   &ShapeLocation(
    Disc Coverage=1
    MaskSpawnGroup=false
    Offset Mode=::NewEnumerator0
    Override Seed=false
    Ring Radius=20
    Shape Primitive=::NewEnumerator4
    UseOwnerScale=true
   )
  ]
  particleUpdateScript [
   &ParticleState()
   &ScaleSpriteSize(Uniform Curve Sprite Scale=DataInterface() {propertyValues {Curve {keys [
        {arriveTangent=1.1687, interpMode=RCIM_Cubic, leaveTangent=1.1687, tangentMode=RCTM_User, time=0, value=0.5}
        {arriveTangent=0.5, interpMode=RCIM_Cubic, leaveTangent=0.5, time=0.2975, value=0.825}
        {tangentMode=RCTM_User, time=1, value=1}
       ]}}})
   &ScaleColor(
    Scale Alpha=&FloatFromCurve(
     FloatCurve=DataInterface() {propertyValues {Curve {keys [
         {arriveTangent=5.9001, interpMode=RCIM_Cubic, leaveTangent=5.9001, tangentMode=RCTM_User, time=0, value=0}
         {arriveTangentWeight=0.1879, interpMode=RCIM_Cubic, leaveTangentWeight=0.2221, tangentMode=RCTM_User, time=0.5, value=1}
         {arriveTangent=-4.3411, interpMode=RCIM_Cubic, leaveTangent=-4.3411, tangentMode=RCTM_User, time=0.95, value=0}
        ]}}}
     Scale Curve=&FNTimeOfDayScalar(Night Scale=0.2, Morning Scale=0.2, Day Scale=0.2, Evening Scale=0.2)
    )
    ScaleA=true
    ScaleRGB=true
   )
   /SetVariables(Particles.0 - Soft Steam SubImageIndex:NiagaraFloat=&FloatFromCurve(FloatCurve=DataInterface() {propertyValues {Curve {keys [
         {arriveTangent=1.824, interpMode=RCIM_Cubic, leaveTangent=1.824, tangentMode=RCTM_User, time=0, value=0}
         {interpMode=RCIM_Cubic, tangentMode=RCTM_User, time=1, value=10}
        ]}}})) {isSetParameters=true}
   &ScaleColor(
    Color Value To Scale=$Particles.Color
    Scale Mode=::NewEnumerator1
    Scale RGBA=&MakeVector4_FromLinearColor(LinearColor=&FNTimeOfDayColor(
      Night Color=<rgba 0.539 0.6941 1 1>
      Morning Color=<rgba 0.9392 0.8785 1 1>
      Day Color=<rgba 1 1 1 1>
      Evening Color=<rgba 1 0.8495 0.8351 1>
     ))
    ScaleRGBA=true
   ) {enabled=false, module=ScaleColor001}
   &DynamicMaterialParameters(
    Index 0 Param 2=1
    Index 0 Param 3=0.5
    Module.Index0Param0Write=true
    Module.Index0Param1Write=true
    Module.Index0Param2Write=true
    Module.Index0Param3Write=true
   )
   &GravityForce(Gravity=<0 0 50>)
   &Drag(Drag=&RandomRangeFloat(Minimum=1, Maximum=3), Use Linear Drag=true, Use Rotational Drag=true)
   &SolveForcesAndVelocity(Clamp Velocity=false, Limit Acceleration=false)
  ]
  rendererClasses [&NiagaraSpriteRendererProperties]
  renderers [{rendererClass=&NiagaraSpriteRendererProperties}]
 }
 Flat_RippleLWE001 {rendererClasses [], renderers []}
 Flat_RippleLWE002 {rendererClasses [], renderers []}
 Flat_RippleLWE003 {rendererClasses [], renderers []}
 Petals {
  emitterUpdateScript [
   &EmitterState(
    Inactive Response=::NewEnumerator1
    Life Cycle Mode=::NewEnumerator1
    Loop Behavior=::NewEnumerator1
    Loop Delay=0.5
    Loop Duration=11
    UseLoopDelay=true
   )
   &SpawnRate(Spawn Probability=0.8, SpawnRate=&RandomRangeFloat(Minimum=3, Maximum=7), Use Spawn Probability=true)
  ]
  particleSpawnScript [
   &InitializeParticle(
    Lifetime Max=3
    Lifetime Min=2
    Lifetime Mode=::NewEnumerator1
    Mass Max=1.5
    Mass Mode=::NewEnumerator3
    Position Offset=<-40.9585 0 70>
    Sprite Rotation Mode=::NewEnumerator2
    Sprite Size Mode=::NewEnumerator1
    Uniform Sprite Size Max=5.5
    Uniform Sprite Size Min=4
    UsePositionOffset=true
    Write Material Random=true
   )
   &ShapeLocation(MaskSpawnGroup=false, Non Uniform Scale=<1 1 1.5>, Override Seed=false, Sphere Radius=75, UseOwnerScale=true)
   &InitialMeshOrientation(Use Rotation Vector=false)
  ]
  particleUpdateScript [
   &ParticleState()
   &GravityForce(Gravity=<0 0 -750>)
   &WindForce(
    Offset By Time:Vector3f=<50 10 -125>
    Scale Min / Max:Vector2f=<0.275 1>
    Speed Min / Max:Vector2f=<0 500>
    Turbulence Scale Mode:ENiagaraWindTurbulenceContributionMode=::NewEnumerator1
    Wind Speed Scale=&RandomRangeFloat(Minimum=0.2, Maximum=0.7)
   )
   &VortexForce(
    UseOriginOffset=false
    Vortex Force Amount=&FloatFromCurve(
     FloatCurve=DataInterface() {propertyValues {Curve {
        defaultValue=0
        keys [
         {interpMode=RCIM_Cubic, tangentMode=RCTM_User, time=0.7, value=0}
         {arriveTangent=3.9004, interpMode=RCIM_Cubic, leaveTangent=3.9004, tangentMode=RCTM_User, time=0.7492, value=1}
         {arriveTangent=-3.4287, interpMode=RCIM_Cubic, leaveTangent=-3.4287, tangentMode=RCTM_User, time=0.7809, value=1}
         {tangentMode=RCTM_User, time=0.8207, value=0}
        ]
       }}}
     CurveIndex=$Emitter.NormalizedLoopAge
     Scale Curve=-4008.7615
    )
   ) {enabled=false}
   &AerodynamicDrag(
    Aerodynamic Drag:NiagaraFloat=&RandomRangeFloat(Minimum=0.3333, Maximum=1)
    Aerodynamic Rotational Drag=0.15
    Lift Contribution=0.75
    Pivot Offset Maximum:Vector3f=<5 0.5 1>
    Pivot Offset Minimum:Vector3f=<-5 -0.5 -1>
    Rotation Strength:NiagaraFloat=&RandomRangeFloat(Minimum=1.8, Maximum=3.2)
    Use Linear Drag=true
    Use Rotational Drag=true
   )
   &Collision(
    Age Colliding Particles=false
    Collision Enabled=&Boolean_OrOperation(
     Bool1=&Boolean_OrOperation(
      Bool1=&IsPlatformSetActive(New Platform Set=DataInterface() {propertyValues {Platforms {qualityLevelMask=-8}}})
      Bool2=$User.bIsFrontEndPreview
     )
     Bool2=false
    )
    Friction=2
    Friction During a Bounce=0.1
    KillOnCollision=false
    Module.Control Rest State Via Surface Normal=false
    Module.EnableMaxCollisionCount=false
    Module.Randomize Collision Normal Vector=false
    Radius Calculation Type=::NewEnumerator1
    Restitution=0.05
   ) {module=Collision002}
   &SolveForcesAndVelocity(Clamp Velocity=false, Limit Acceleration=false)
   &ScaleSpriteSize(
    Uniform Curve Scale=&FloatFromCurve(FloatCurve=DataInterface() {propertyValues {Curve {keys [<at 0.85 1>,<at 1 0>]}}}, CurveIndex=$Emitter.NormalizedLoopAge)
    Uniform Curve Sprite Scale=DataInterface() {propertyValues {Curve {keys [
        {arriveTangent=14.5835, interpMode=RCIM_Cubic, leaveTangent=14.5835, tangentMode=RCTM_User, time=0, value=0}
        {arriveTangent=1.1111, interpMode=RCIM_Cubic, leaveTangent=1.1111, time=0.1, value=1}
        {arriveTangent=-1.1111, interpMode=RCIM_Cubic, leaveTangent=-1.1111, time=0.9, value=1}
        {arriveTangent=-8.7113, interpMode=RCIM_Cubic, leaveTangent=-8.7113, tangentMode=RCTM_User, time=1, value=0}
       ]}}}
   )
   &AlignSpriteToMeshOrientation(SetAlignment=true, SetFacing=true)
  ]
  rendererClasses [&NiagaraSpriteRendererProperties]
  renderers [{rendererClass=&NiagaraSpriteRendererProperties}]
 }
 Petals004 {
  emitterUpdateScript [
   &EmitterState(Life Cycle Mode=::NewEnumerator1, Loop Behavior=::NewEnumerator1, Loop Delay=0.0333, Loop Duration=11.5667, UseLoopDelay=true)
   &SpawnRate(
    SpawnRate=&FloatFromCurve(
     FloatCurve=DataInterface() {propertyValues {Curve {keys [<at 0 1 RCIM_Constant>,<at 0.1558 0 RCIM_Constant>,<at 0.3549 2 RCIM_Constant>,<at 0.5971 0 RCIM_Constant>,<at 0.8613 1 RCIM_Constant>,<at 1 0>]}}}
     CurveIndex=$Emitter.NormalizedLoopAge
    )
    Use Spawn Probability=false
   )
   &SpawnBurst_Instantaneous(Spawn Time=0.5667, Use Spawn Probability=false, UseLoopCountLimit=false) {enabled=false, module=SpawnBurst_Instantaneous005}
  ]
  particleSpawnScript [
   &InitializeParticle(
    Lifetime=11
    Mesh Scale Mode=::NewEnumerator1
    Mesh Uniform Scale Max=0.35
    Mesh Uniform Scale Min=0.2
    Position Offset=<-5 0 60>
    Sprite Size Mode=::NewEnumerator3
    Uniform Sprite Size=&RandomRangeFloat(Minimum=2, Maximum=3.5)
    UsePositionOffset=true
    Write Material Random=true
   )
   &ShapeLocation(
    Box Size=<30 30 5>
    MaskSpawnGroup=false
    Non Uniform Scale=<1 1 0.2>
    Override Seed=false
    Shape Primitive=::NewEnumerator2
    Surface Only=false
    UseOwnerScale=true
   )
   /Game/Sprite/FN_SpriteAlignmentAndFacing(SpriteAlignment:Vector3f=<1 0 1>, SpriteFacing:Vector3f=<0 0 1>)
   &InitialMeshOrientation(Rotation=<0 0 0.325>, Use Rotation Vector=false)
   &AddVelocity(
    Module.UseConeFalloff=false
    Module.UseVelDistribution=false
    Override Seed=false
    Velocity Mode=::NewEnumerator2
    Velocity Speed=&RandomRangeFloat(Minimum=20, Maximum=50)
   )
  ]
  particleUpdateScript [
   &ParticleState()
   &ScaleSpriteSize(
    Uniform Curve Scale=2
    Uniform Curve Sprite Scale=DataInterface() {propertyValues {Curve {keys [
        {arriveTangent=1.1111, interpMode=RCIM_Cubic, time=0.1, value=1}
        {arriveTangent=-0.0808, interpMode=RCIM_Cubic, leaveTangent=-0.0808, tangentMode=RCTM_User, time=0.9, value=1}
        {arriveTangent=-8.7113, interpMode=RCIM_Cubic, leaveTangent=-8.7113, tangentMode=RCTM_User, time=1, value=0}
       ]}}}
   )
   &Collision(
    Advanced Aging Rate=0.5
    Age Colliding Particles=false
    Analytical Collision Plane Position 1:Vector3f=<0 0 7.5>
    Analytical Collision Plane Space:ENiagaraCoordinateSpace=::Local
    CPU Collision Type=::NewEnumerator3
    Collision Enabled=&self:ScratchDynamicInput_04()
    Friction=0.1
    Kill On Collision=0
    KillOnCollision=false
    Mesh Dimensions:Vector3f=<0.3 0.3 0.3>
    Module.Control Rest State Via Surface Normal=false
    Module.EnableMaxCollisionCount=false
    Module.Randomize Collision Normal Vector=false
    Particle Radius Scale=0.5
    Perform 2 Analytical Collision Queries=false
    Radius Calculation Type=::NewEnumerator1
    Restitution=0
   )
   &Drag(
    Drag=&FloatFromCurve(
     FloatCurve=DataInterface() {propertyValues {Curve {
        defaultValue=1
        keys [
         <at 0.05 0.5 RCIM_Cubic>
         {arriveTangent=5, interpMode=RCIM_Cubic, leaveTangent=5, time=0.07, value=1}
         {arriveTangent=-1.2265, interpMode=RCIM_Cubic, leaveTangent=-1.2265, time=0.15, value=1}
         <at 0.3961 0.6 RCIM_Cubic>
         <at 0.7937 1 RCIM_Cubic>
        ]
       }}}
     Scale Curve=&RandomRangeFloat(Minimum=10, Maximum=16)
    )
    Use Linear Drag=true
    Use Rotational Drag=true
   )
   &CurlNoiseForce(
    Bias Noise=false
    Noise Frequency=20
    Noise Strength=&MakeCustomFloatFromBool(
     Bool=$Particles.HasCollided
     True Float=200
     False Float=&FloatFromCurve(
      FloatCurve=DataInterface() {propertyValues {Curve {keys [<at 0 0 RCIM_Cubic>,<at 0.1 1 RCIM_Cubic>]}}}
      Scale Curve=&RandomRangeFloat(Minimum=250, Maximum=500)
     )
    )
    Pan Noise=false
   )
   &Timeline(Play=$Particles.HasCollided, Play Rate=2.25)
   &MeshRotationForce(Rotation=&VectorFromFloat(Value=&Lerp_Float(A=&RandomRangeFloat(Minimum=-30, Maximum=30), B=0, Alpha=$Particles.Timeline.NormalizedTime)))
   &GravityForce(Gravity=&RandomRangeVector(Minimum=<0 0 -650>, Maximum=<1 1 -750>))
   &VortexForce(UseOriginOffset=false)
   &SolveForcesAndVelocity(Clamp Velocity=false, Limit Acceleration=false)
   &ScaleColor(Scale Alpha=0, ScaleA=true, ScaleRGB=true)
   &SpriteRotationRate(Rotation Rate=&self:ScratchDynamicInput_05(rotation:NiagaraFloat=3))
   /Niagara/Modules/Collision/AlignParticlesWithCollisionPlane(
    Allow Alignment on Y=false
    Allow Alignment on Z=false
    Sprite Relative Alignment Vector:Vector3f=<1 0 0>
    Sprite Relative Facing Vector:Vector3f=<0 1 0>
   )
   &SpriteFacingAndAlignment(
    Sprite Facing=&MakeVector(X=&Lerp_Float(A=1, B=0, Alpha=$Particles.Age), Y=0, Z=0)
    WriteAlignment=true
    WriteFacing=true
   )
   &AlignSpriteToMeshOrientation(SetAlignment=true, SetFacing=true)
   &KillParticlesInVolume(Invert Volume=true, Origin Offset:Vector3f=<0 0 65>)
  ]
  rendererClasses [&NiagaraSpriteRendererProperties]
  renderers [{rendererClass=&NiagaraSpriteRendererProperties}]
 }
 Ripples_merged {
  emitterSpawnScript [/SetVariables(Emitter.Innercolor:LinearColor=/Game/Effects/Niagara/DynamicInputs/Logic/FN_PickLinearColorDependOnPlatform(
     TargetPlatform:NiagaraDataInterfacePlatformSet=DataInterface() {propertyValues {Platforms {
        qualityLevelMask=-1
        deviceProfileStates [
         {profileName=Switch, qualityLevelMask=4294967295, setQualityLevelMask=4294967295}
         {profileName=Mobile, qualityLevelMask=4294967295, setQualityLevelMask=4294967295}
        ]
       }}}
     TargetPlatformActive:LinearColor=&FNTimeOfDayColor(
      Night Color=<rgba 0.0561 0.0452 0.0782 0.5>
      Morning Color=<rgba 0.0833 0.0172 0.0132 0.5>
      Day Color=<rgba 0.1823 0.0679 0.022 1>
      Evening Color=<rgba 0.125 0.0223 0.0187 0.5>
     )
     TargetPlatformNoActive:LinearColor=&FNTimeOfDayColor(
      Night Color=<rgba 0.0956 0.0988 0.2086 1>
      Morning Color=<rgba 0.2083 0.1024 0.097 1>
      Day Color=<rgba 0.3385 0.1664 0.1577 1>
      Evening Color=<rgba 0.2083 0.1024 0.097 1>
     )
    )) {isSetParameters=true}]
  emitterUpdateScript [
   &EmitterState(Life Cycle Mode=::NewEnumerator1, Loop Behavior=::NewEnumerator1, Loop Delay=8, Loop Duration=12, UseLoopDelay=false)
   &SpawnRate(SpawnGroup=3, SpawnRate=2, Use Spawn Probability=false)
   &SpawnBurst_Instantaneous(Age:NiagaraFloat=$Emitter.Age, Spawn Count=2, Spawn Group=2, Spawn Time=9.4, Use Spawn Probability=false, UseLoopCountLimit=false) {module=SpawnBurst_Instantaneous003}
   &SpawnBurst_Instantaneous(Age:NiagaraFloat=$Emitter.Age, Spawn Count=2, Spawn Group=2, Spawn Time=10.75, Use Spawn Probability=false, UseLoopCountLimit=false)
   &SpawnBurst_Instantaneous(Age:NiagaraFloat=$Emitter.Age, Spawn Count=4, Spawn Group=1, Spawn Time=8.1, Use Spawn Probability=false, UseLoopCountLimit=false) {module=SpawnBurst_Instantaneous002}
   &SpawnBurst_Instantaneous(Age:NiagaraFloat=$Emitter.Age, Spawn Count=4, Spawn Time=7.8, Use Spawn Probability=true, UseLoopCountLimit=false) {module=SpawnBurst_Instantaneous001}
  ]
  particleSpawnScript [
   /SetVariables(Particles.VisibilityTag=&SpawnGroup()) {isSetParameters=true}
   &InitializeParticle(
    Color Mode=::NewEnumerator0
    Lifetime=&MergeHelper_FloatSwitch(
     0=&RandomRangeFloat(Minimum=0.7, Maximum=1.7)
     1=&RandomRangeFloat(Minimum=0.7, Maximum=1.7)
     2=0.65
     3=&RandomRangeFloat(Minimum=1, Maximum=2)
     Index=&SpawnGroup()
     Num=::NewEnumerator2
    )
    Position Offset=<0 0 3>
    Sprite Size Mode=::NewEnumerator3
    Uniform Sprite Size=45
    UsePositionOffset=true
    Write Material Random=true
   )
  ]
  particleUpdateScript [
   &ParticleState()
   &ScaleSpriteSize()
   &SkeletalMeshLocation(
    Bone Sampling Mode=::NewEnumerator4
    MaskSpawnGroup=false
    Sampled Position Offset=&MergeHelper_VectorSwitch(0=<0 -8 5>, 1=<0 15 5>, 2=<0 10 2>, 3=<0 0 1.2>, Index=&SpawnGroup(), Num=::NewEnumerator2)
    Sampled Position Offset Coordinate Space:ENiagaraSkelMeshTransforms=::NewEnumerator2
    Skeletal Mesh=$System.cup skeletal mesh
   )
   /SetVariables(Particles.SpriteFacing=&TransformVector(Vector=<0 0 1>, Source Space=::World, Destination Space=::Local)) {isSetParameters=true}
   &ScaleColor(
    Scale Alpha=&FloatFromCurve(
     FloatCurve=DataInterface() {propertyValues {Curve {
        defaultValue=0
        keys [
         {interpMode=RCIM_Cubic, tangentMode=RCTM_User, time=0, value=0}
         {interpMode=RCIM_Cubic, tangentMode=RCTM_User, time=0.2575, value=1}
         {tangentMode=RCTM_User, time=0.9397, value=0}
        ]
       }}}
     Scale Curve=&FNTimeOfDayScalar(Night Scale=6, Morning Scale=5, Day Scale=5, Evening Scale=0.5)
    )
    ScaleA=true
    ScaleRGB=true
   )
   /SetVariables(
    Particles.3 - Constant Ripples:LinearColor=&MakeLinearColorFromVectorAndFloat(
     "Vector (RGB)"=<1 1 1>
     "Float (Alpha)"=&FloatFromCurve(
      FloatCurve=DataInterface() {propertyValues {Curve {
         defaultValue=0
         keys [
          {arriveTangent=9.0725, interpMode=RCIM_Cubic, leaveTangent=9.0725, tangentMode=RCTM_User, time=0, value=0}
          {arriveTangent=-2.5249, interpMode=RCIM_Cubic, leaveTangent=-2.5249, tangentMode=RCTM_User, time=0.2308, value=0.9847}
          {arriveTangent=-0.0405, interpMode=RCIM_Cubic, leaveTangent=-0.0405, tangentMode=RCTM_User, time=1, value=0}
         ]
        }}}
      Scale Curve=&FNTimeOfDayScalar(Night Scale=0.5, Morning Scale=0.5, Day Scale=0.5, Evening Scale=0.5)
     )
    )
    Particles.3 - constant ripple spprite size:Vector2f=&Multiply_Vector2D_ByFloat(
     Float=&Waveform("Waveform [1]"=::NewEnumerator0, "[1] Phase Scale"=0.3, "[1] Frequency"=1, "[1] Amplitude Min/Max"=<0.75 1>)
     Vector2D=&Vector2DFromFloat(Value=75)
    )
   ) {isSetParameters=true}
   &DynamicMaterialParameters(
    Index 0 Param 1=&FloatFromCurve(
     FloatCurve=DataInterface() {propertyValues {Curve {keys [{arriveTangent=1.3374, interpMode=RCIM_Cubic, leaveTangent=1.3374, tangentMode=RCTM_User, time=0, value=1}]}}}
     CurveIndex=$Particles.Age
    )
    Index 0 Param 2=&RandomRangeFloat(Minimum=0.05, Maximum=0.15)
    Index 0 Param 3=&FloatFromCurve(
     FloatCurve=DataInterface() {propertyValues {Curve {keys [
         {arriveTangent=4.9025, interpMode=RCIM_Cubic, leaveTangent=4.9025, tangentMode=RCTM_User, time=0, value=0}
         {arriveTangent=1.4945, interpMode=RCIM_Cubic, tangentMode=RCTM_Break, time=0.5796, value=2}
        ]}}}
     CurveIndex=$Particles.Age
     Scale Curve=0.05
    )
    Index 0 Param 4=&FloatFromCurve(
     FloatCurve=DataInterface() {propertyValues {Curve {keys [
         {arriveTangent=-3.3445, interpMode=RCIM_Cubic, leaveTangent=-3.3445, tangentMode=RCTM_User, time=0, value=0}
         {arriveTangent=-1.0702, interpMode=RCIM_Cubic, leaveTangent=-1.0702, time=0.0558, value=0.2806}
         {arriveTangent=-1.1268, interpMode=RCIM_Cubic, leaveTangent=-1.2066, tangentMode=RCTM_Break, time=0.3394, value=-0.3632}
         {arriveTangent=-0.4773, interpMode=RCIM_Cubic, tangentMode=RCTM_Break, time=1.0134, value=-0.5457}
        ]}}}
     CurveIndex=$Particles.Age
    )
    Index 1 Param 1=0.8
    Index 1 Param 2=0.45
    Index 1 Param 4=3
    Module.Param0WriteEnabled=true
    Module.Param0WriteEnabled1=true
    Module.Param1WriteEnabled=true
    Module.Param1WriteEnabled1=true
    Module.Param2WriteEnabled=true
    Module.Param2WriteEnabled1=true
    Module.Param3WriteEnabled=true
    Module.Param3WriteEnabled1=true
    Write Parameter Index 1=true
   ) {module=DynamicMaterialParameters001}
   /SetVariables(Particles.1 - feet ripple size:Vector2f=&Multiply_Vector2D_ByFloat(
     Float=&FloatFromCurve(FloatCurve=DataInterface() {propertyValues {Curve {defaultValue=1, keys [{tangentMode=RCTM_User, time=0, value=0}, {tangentMode=RCTM_User, time=1, value=1}]}}})
     Vector2D=<55 45>
    )) {isSetParameters=true}
   /SetVariables(Particles.2 - head peak color:LinearColor=&MakeLinearColorFromVectorAndFloat(
     "Vector (RGB)"=<1 1 1>
     "Float (Alpha)"=&FloatFromCurve(FloatCurve=DataInterface() {propertyValues {Curve {
         defaultValue=1
         keys [
          {arriveTangent=9.0725, interpMode=RCIM_Cubic, leaveTangent=9.0725, tangentMode=RCTM_User, time=0, value=0}
          {arriveTangent=-2.5249, interpMode=RCIM_Cubic, leaveTangent=-2.5249, tangentMode=RCTM_User, time=0.1408, value=0.9847}
          {arriveTangent=-0.0405, interpMode=RCIM_Cubic, leaveTangent=-0.0405, tangentMode=RCTM_User, time=1, value=0}
         ]
        }}})
    )) {isSetParameters=true}
  ]
  rendererClasses [&NiagaraSpriteRendererProperties]
  renderers [
   {rendererClass=&NiagaraSpriteRendererProperties}
   {rendererClass=&NiagaraSpriteRendererProperties, rendererIndex=1}
   {rendererClass=&NiagaraSpriteRendererProperties, rendererIndex=2}
   {rendererClass=&NiagaraSpriteRendererProperties, rendererIndex=3}
  ]
 }
 Steam {
  emitterUpdateScript [
   &EmitterState(Life Cycle Mode=::NewEnumerator1, Loop Behavior=::NewEnumerator1, Loop Duration=11, UseLoopDelay=false)
   &SpawnBurst_Instantaneous(Spawn Count=10, Use Spawn Probability=false, UseLoopCountLimit=false) {enabled=false}
   &SpawnRate(SpawnRate=25, Use Spawn Probability=false)
  ]
  particleSpawnScript [
   &InitializeParticle(
    Color=&MakeLinearColorFromVectorAndFloat("Vector (RGB)"=&MakeVectorFromLinearColorRGB(Color=$System.SteamColor), "Float (Alpha)"=0.4)
    Lifetime Mode=::NewEnumerator1
    Position Offset=<0 0 10>
    Sprite Rotation Mode=::NewEnumerator2
    Sprite Size Max=<55 65>
    Sprite Size Min=<30 40>
    Sprite Size Mode=::NewEnumerator2
    UsePositionOffset=true
    Write Material Random=true
   )
   &AddVelocity(Velocity=&RandomRangeVector(Minimum=<-5 -5 5>, Maximum=<5 5 35>, Evaluation Type=::NewEnumerator1))
   &ShapeLocation(
    Cylinder Height=1.2
    Cylinder Radius=20
    Hemicircle Internal Cap=false
    Hemicircle X=true
    MaskSpawnGroup=false
    Override Seed=false
    Rotation Axis=<4.1206 0 0>
    Rotation Mode=::NewEnumerator1
    Shape Primitive=::NewEnumerator1
    Surface Only=true
    Use Endcaps In Surface Only Mode=false
    UseOwnerScale=true
   ) {enabled=false}
   &ShapeLocation(
    Cylinder Height=1.2
    Cylinder Radius=25
    Hemicircle X=true
    MaskSpawnGroup=false
    Offset Mode=::NewEnumerator0
    Override Seed=false
    Rotation Angle=0.25
    Rotation Axis=<0 0 1>
    Rotation Mode=::NewEnumerator1
    Shape Primitive=::NewEnumerator1
    Surface Only=true
    Use Endcaps In Surface Only Mode=false
    UseOwnerScale=true
   ) {module=ShapeLocation001}
  ]
  particleUpdateScript [
   &ParticleState()
   &ScaleSpriteSize(Scale Factor=&Vector2DFromFloat(Value=&FloatFromCurve(FloatCurve=DataInterface() {propertyValues {Curve {keys [
          {arriveTangent=1.8751, interpMode=RCIM_Cubic, leaveTangent=1.8751, tangentMode=RCTM_User, time=0, value=0.25}
          {arriveTangent=0.72, interpMode=RCIM_Cubic, leaveTangent=0.72, tangentMode=RCTM_User, time=1, value=1.25}
         ]}}})))
   &SubUVAnimation(
    Sub UV Play Rate:NiagaraFloat=&RandomRangeFloat(Minimum=0.01, Maximum=0.1)
    SubUV Animation Mode:ENiagara_SubUVLookupMode=::NewEnumerator3
    SubUV Start Time Offset:NiagaraFloat=&RandomRangeFloat(Minimum=0, Maximum=63)
   )
   &DynamicMaterialParameters(
    Index 0 Param 1=&FloatFromCurve(FloatCurve=DataInterface() {propertyValues {Curve {keys [<at 0 -0.25>,<at 1 -1>]}}})
    Index 0 Param 4=&OneMinusFloat(Float=$Particles.NormalizedAge)
    Module.Param0WriteEnabled=true
    Module.Param1WriteEnabled=false
    Module.Param2WriteEnabled=false
    Module.Param3WriteEnabled=true
   )
   &ScaleColor(Scale Alpha=$System.FadeIn, ScaleA=true, ScaleRGB=false) {module=ScaleColor001}
   &ScaleColor(
    Scale Alpha=&FloatFromCurve(
     FloatCurve=DataInterface() {propertyValues {Curve {keys [
         {arriveTangent=5.2427, interpMode=RCIM_Cubic, leaveTangent=5.2427, tangentMode=RCTM_User, time=0, value=0}
         {arriveTangent=-1.0898, interpMode=RCIM_Cubic, leaveTangent=-1.0898, tangentMode=RCTM_User, time=0.3, value=1}
         {arriveTangent=-0.1231, interpMode=RCIM_Cubic, leaveTangent=-0.1231, tangentMode=RCTM_User, time=1, value=0}
        ]}}}
     Scale Curve=&RandomRangeFloat(
      Minimum=0.4
      Maximum=&FNTimeOfDayScalar(Night Scale=0.8, Morning Scale=0.7, Day Scale=0.7, Evening Scale=0.7)
     )
    )
    ScaleA=true
    ScaleRGB=false
   )
   &Drag(
    Drag=&FloatFromCurve(
     FloatCurve=DataInterface() {propertyValues {Curve {
        defaultValue=1
        keys [
         {interpMode=RCIM_Cubic, tangentMode=RCTM_User, time=0, value=0}
         {arriveTangent=0.874, interpMode=RCIM_Cubic, leaveTangent=0.874, tangentMode=RCTM_User, time=1, value=1}
        ]
       }}}
     Scale Curve=&RandomRangeFloat(Minimum=1, Maximum=6)
    )
    Use Linear Drag=true
    Use Rotational Drag=false
   )
   &SolveForcesAndVelocity(Clamp Velocity=false, Limit Acceleration=false)
  ]
  rendererClasses [&NiagaraSpriteRendererProperties]
  renderers [{rendererClass=&NiagaraSpriteRendererProperties}]
 }
}
summary {
 systemName=NS_Emote_SnugLoaf_steam
 emitters [
  {emitterName=Steam, rendererClasses [&NiagaraSpriteRendererProperties]}
  {emitterName=Flat_RippleLWE001, rendererClasses []}
  {emitterName=Petals, rendererClasses [&NiagaraSpriteRendererProperties]}
  {emitterName=Flat_RippleLWE002, rendererClasses []}
  {emitterName=Flat_RippleLWE003, rendererClasses []}
  {emitterName=Petals004, rendererClasses [&NiagaraSpriteRendererProperties]}
  {emitterName=Billow_steam001, rendererClasses [&NiagaraSpriteRendererProperties]}
  {emitterName=Ripples_merged, rendererClasses [&NiagaraSpriteRendererProperties]}
 ]
 userVariables [
  User.bIsFrontEnd {type=NiagaraBool}
  User.bIsFrontEndPreview {type=NiagaraBool}
  User.latte_ripple_alpha {type=NiagaraFloat, defaultValue=4.2534, description=""}
  User.RibbonIDOffset {type=NiagaraInt32}
 ]
}
